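; Debug actor that reports Chaos destruction activity through a ChaosDestructionListener.
; Its collision, breaking and trailing event handlers each format the length of the incoming event array into a "Number of ... Events this frame" message.
; Each message prints to screen and log for one frame, using world delta seconds as its duration.
; BeginPlay, overlap, Tick and the construction script are empty.
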(event EventBeginPlay)

(event Collision|EventActorBeginOverlap
  (params (OtherActor :type Actor%20Object%20Reference)))

(event EventTick
  (params (DeltaSeconds :type Float%20%28single-precision%29)))

(event OnCollisionEvents(ChaosDestructionListener)
  (params (CollisionEvents :type Array%20of%20Chaos%20Collision%20Event%20Data%20Structures))
  (Development|PrintText (Utilities|Text|FormatText "NSLOCTEXT(\"[804974D54C54070216E7688786D3BA35]\", \"44189E3D47C9294BEC312E8529A7C5B9\", \"Number of Collision Events this frame: {num}\")" (Utilities|Array|Length CollisionEvents)) true true "(R=0.000000,G=0.660000,B=1.000000,A=1.000000)" (Utilities|Time|GetWorldDeltaSeconds)))

(event OnBreakingEvents(ChaosDestructionListener)
  (params (BreakingEvents :type Array%20of%20Chaos%20Breaking%20Event%20Data%20Structures))
  (Development|PrintText (Utilities|Text|FormatText "NSLOCTEXT(\"[804974D54C54070216E7688786D3BA35]\", \"1986BD5346D26C3398AB39B350E94D8B\", \"Number of Breaking Events this frame: {num}\")" (Utilities|Array|Length BreakingEvents)) true true "(R=0.000000,G=0.660000,B=1.000000,A=1.000000)" (Utilities|Time|GetWorldDeltaSeconds)))

(event OnTrailingEvents(ChaosDestructionListener)
  (params (TrailingEvents :type Array%20of%20Chaos%20Trailing%20Event%20Data%20Structures))
  (Development|PrintText (Utilities|Text|FormatText "NSLOCTEXT(\"[804974D54C54070216E7688786D3BA35]\", \"6030DCF64977449E27EEEDAFBC1CCE05\", \"Number of Trailing Events this frame: {num}\")" (Utilities|Array|Length TrailingEvents)) true true "(R=0.000000,G=0.660000,B=1.000000,A=1.000000)" (Utilities|Time|GetWorldDeltaSeconds)))

(fn ConstructionScript)
